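# Physics asset for the HighTower Grape backpack cosmetic, defining bodies and joints for its dynamic hair and branch bones.
# Small single-capsule bodies cover three hair bones, a right branch bone, and the head and right lower arm they attach to.
# The hair bones are constrained to the head with very tight angular limits.
# The branch is constrained to the lower arm with wide swing but narrow twist.
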PhysicsAsset /BRCosmetics/Accessories/FORT_Backpacks/Backpack_M_MED_HighTower_Grape/Meshes/M_MED_HighTower_Grape_Pack_Dyn
bodies {
 dyn_branch_r [{shapeType=Capsule, length=1.2237, radius=0.505, center=<1.4351 0 0>, rotation=<rot 90 0 0>}]
 dyn_hair_A [{shapeType=Capsule, length=1.2237, radius=0.505, center=<0.1 0 0>, rotation=<rot 90 0 0>}]
 dyn_hair_B [{shapeType=Capsule, length=1.2237, radius=0.505, center=<0.1 0 0>, rotation=<rot 90 0 0>}]
 dyn_hair_C [{shapeType=Capsule, length=1.2237, radius=0.505, center=<0.1 0 0>, rotation=<rot 90 0 0>}]
 head [{shapeType=Capsule, length=0.5, radius=0.5}]
 lowerarm_r [{shapeType=Capsule, length=0.5, radius=0.5}]
}
constraints [
 {bone1Name=dyn_hair_A, bone2Name=head, swing1LimitDegrees=1, swing2LimitDegrees=1, twistLimitDegrees=1}
 {bone1Name=dyn_hair_B, bone2Name=head, swing1LimitDegrees=1, swing2LimitDegrees=1, twistLimitDegrees=1}
 {bone1Name=dyn_hair_C, bone2Name=head, swing1LimitDegrees=1, swing2LimitDegrees=1, twistLimitDegrees=1}
 {bone1Name=dyn_branch_r, bone2Name=lowerarm_r, swing1LimitDegrees=60, swing2LimitDegrees=60, twistLimitDegrees=5}
]
